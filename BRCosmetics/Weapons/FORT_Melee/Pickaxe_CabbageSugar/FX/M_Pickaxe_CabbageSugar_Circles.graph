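# Translucent unlit Niagara sprite material for the CabbageSugar pickaxe FX that draws a procedural circle tinted by particle color.
# Sphere masks over UVs build the disc, and the difference of two radii darkens a thin rim before emissive adjustment.
# An off-center soft-edged mask is subtracted from the disc to dissolve it, multiplied by particle alpha.
# Dynamic parameters drive circle and dissolve radii, so no textures are needed.
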Material /BRCosmetics/Weapons/FORT_Melee/Pickaxe_CabbageSugar/FX/M_Pickaxe_CabbageSugar_Circles
graph {
 properties {
  bAutomaticallySetUsageInEditor=false
  bUsedWithNiagaraSprites=true
  blendMode=BLEND_Translucent
  pixelDepthOffsetMode=PDOM_Legacy
  shadingModel=MSM_Unlit
  usedShadingModels=Unlit
 }
 nodes [
  n1 = ParticleColor()
  n2 = TextureCoordinate()
  n3 = Constant() {r=0.5}
  n4 = DynamicParameter() {defaultValue=<rgba 1 0 1 1>, paramNames [Radius, DissolveRadius, Param3, Param4]}
  n5 = Multiply(A=Constant() {r=0.5}, B=@n4.Radius)
  n6 = Constant() {r=1}
  n7 = SphereMask(A=@n2, B=@n3, Radius=@n5, Hardness=@n6)
  n8 = SphereMask(A=@n2, B=@n3, Radius=Subtract(A=@n5) {constB=0.04}, Hardness=@n6)
  MaterialFunctionCall(Emissive=LinearInterpolate(A=@n1.RGB, B=Multiply(A=@n1.RGB) {constB=0.2}, Alpha=Subtract(A=@n7, B=@n8))) {materialFunction=&MF_FX_EmissiveAdjustments}
  n9 = SphereMask(A=@n2, B=Constant2Vector() {g=0.65, r=0.6}, Radius=@n4.DissolveRadius, Hardness=Constant() {r=0.92})
  Multiply(A=@n1.A, B=Saturate(None=Subtract(A=@n7, B=@n9)))
 ]
}
