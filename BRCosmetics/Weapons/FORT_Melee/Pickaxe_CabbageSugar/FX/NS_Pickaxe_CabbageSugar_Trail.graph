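# Niagara system for the green ribbon swing trail on the CabbageSugar pickaxe.
# Its MainTrails emitter uses Fortnite anim-trail modules to spawn short-lived particles between two skeletal mesh bones or sockets, drawn by two ribbon renderers.
# Particle color and a separate SoftColor attribute are tinted by system time-of-day values and fade out over curves scaled by a system swing-fade value.
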NiagaraSystem /BRCosmetics/Weapons/FORT_Melee/Pickaxe_CabbageSugar/FX/NS_Pickaxe_CabbageSugar_Trail
dependencies {
 usedDataInterfaces [&NiagaraDataInterfaceCurve, &NiagaraDataInterfaceSkeletalMesh, &NiagaraDataInterfaceColorCurve]
 usedDynamicInputs [&ColorFromCurve, &FNTimeOfDayScalar]
 usedModules [
  &self:SystemSpawnScript.SystemScriptSource.SystemScriptGraph.NiagaraNodeAssignment_0.SetVariables_#1
  &SystemState
  &self:SystemSpawnScript.SystemScriptSource.SystemScriptGraph.NiagaraNodeAssignment_1.SetVariables_#2
  &self:SystemSpawnScript.SystemScriptSource.SystemScriptGraph.NiagaraNodeAssignment_3.SetVariables_#3
 ]
}
emitters {MainTrails {
  emitterSpawnScript [&FN_RibbonIDOffsetUtility()]
  emitterUpdateScript [
   &EmitterState()
   &FN_SpawnAnimTrail()
  ]
  particleSpawnScript [
   &FN_InitializeAnimTrailBoneIndecies(First Bone or Socket Index=1, Second Bone or Socket Index=0)
   &FN_InitializeAnimTrail(
    Color=&MultiplyLinearColor(
     Linear Color=<rgba 0.3587 2 0.5487 1>
     Scale Factor=&MakeVector4_FromLinearColor(LinearColor=$System.TODColor)
    )
    Lifetime=0.11
    SkeletalMesh=$System.Skeletal Mesh
   )
  ]
  particleUpdateScript [
   &FN_UpdateAnimTrail(SkeletalMesh=$System.Skeletal Mesh)
   &ScaleColor(
    Scale Alpha=&Multiply_Float(
     A=&FloatFromCurve(
      FloatCurve=DataInterface() {propertyValues {Curve {keys [{interpMode=RCIM_Cubic, leaveTangent=-2.522, tangentMode=RCTM_Break, time=0, value=1}, <at 0.6 0>]}}}
      Scale Curve=$System.SwingGeneralFade
     )
     B=0.5
    )
    Scale RGB=<0.8 0.8 0.8>
    ScaleA=true
    ScaleRGB=true
   )
   &ScaleColor(Scale RGB=&VectorFromFloat(Value=$System.TODFloat), ScaleA=false, ScaleRGB=true) {module=ScaleColor001}
   /SetVariables(Particles.SoftColor:LinearColor=&MakeLinearColorFromVectorAndFloat(
     "Vector (RGB)"=&Multiply_VectorByFloat(
      Vector=&Multiply_Vector(
       A=&MakeVectorFromLinearColorRGB(Color=<rgba 0.1714 0.956 0.2623 1>)
       B=&MakeVectorFromLinearColorRGB(Color=$System.TODColor)
      )
      Float=$System.TODFloat
     )
     "Float (Alpha)"=&Multiply_Float(
      A=&FloatFromCurve(
       FloatCurve=DataInterface() {propertyValues {Curve {keys [{interpMode=RCIM_Cubic, leaveTangent=-1.7551, tangentMode=RCTM_Break, time=0, value=1}, <at 0.6 0>]}}}
       Scale Curve=$System.SwingGeneralFade
      )
      B=0.8
     )
    )) {isSetParameters=true}
  ]
  rendererClasses [&NiagaraRibbonRendererProperties]
  renderers [
   {rendererClass=&NiagaraRibbonRendererProperties}
   {rendererClass=&NiagaraRibbonRendererProperties, rendererIndex=1}
  ]
 }}
summary {
 systemName=NS_Pickaxe_CabbageSugar_Trail
 emitters [{emitterName=MainTrails, rendererClasses [&NiagaraRibbonRendererProperties]}]
 userVariables [User.RibbonIDOffset {type=NiagaraInt32}]
}
