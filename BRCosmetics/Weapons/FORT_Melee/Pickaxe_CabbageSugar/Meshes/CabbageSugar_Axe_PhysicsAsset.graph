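# Physics asset for the CabbageSugar two-handed pickaxe cosmetic mesh, a Battle Royale melee weapon.
# It holds a single capsule body on the Melee_TwoHanded_Handle bone, offset along the handle and rotated.
# With only one body and no constraints, it gives the weapon a simple collision shape rather than any articulated simulation.
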PhysicsAsset /BRCosmetics/Weapons/FORT_Melee/Pickaxe_CabbageSugar/Meshes/CabbageSugar_Axe_PhysicsAsset
bodies {Melee_TwoHanded_Handle [{shapeType=Capsule, length=74.7553, radius=32.6222, center=<0.6819 -0.909 34.8505>, rotation=<rot 0.0705 -90 0.004>}]}
constraints []
